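; Dynamic UI director that shows the Beast Girl HUD scene while the local player's view target has the activation tag.
; On BeginPlay and each view target change, it resolves the viewed pawn and starts async gameplay tag add/remove listeners on that pawn's PlayerState.
; Tag changes pass through a delta gate, so only real state changes add or remove the scene and hide or restore configured HUD elements under the tag's context.
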(event EventBeginPlay
  ;[0.1538 1 0.5728] BeginPlay
  (bind _asfort_player_controller (Utilities|Casting|CastToFortPlayerController (DynamicUIDirector|GetOwningLocalPlayerController)))
  (EventDispatchers|BindEventtoOnViewTargetChanged _asfort_player_controller (delegate Custom|OnViewTargetChanged))
  (CallFunction|SetViewTarget (Pawn|GetControlledPawn _asfort_player_controller)))

(event Custom|OnViewTargetChanged
  (params (PlayerController :type Fort%20Player%20Controller%20Object%20Reference) (OldViewTarget :type Actor%20Object%20Reference) (NewViewTarget :type Actor%20Object%20Reference))
  ;[0.1538 1 0.5728] BeginPlay
  (bind _asfort_player_controller (Utilities|Casting|CastToFortPlayerController NewViewTarget)
    (:then
      (CallFunction|SetViewTarget (Pawn|GetControlledPawn _asfort_player_controller)))
    (:CastFailed
      (bind _aspawn (Utilities|Casting|CastToPawn NewViewTarget))
      (CallFunction|SetViewTarget _aspawn))))

(event Custom|SetViewTarget
  (params (NewPawn :type Pawn%20Object%20Reference))
  (bind _playerstate (Variables|/Script/Engine.Pawn|GetPlayerState NewPawn))
  (Utilities|IsValid NewPawn
    (:Is%20Valid
      (Utilities|IsValid _playerstate
        (:Is%20Valid
          ;[0.1538 1 0.5728] Set Pawn Listeners
          (Ability|Async|EndAction (Variables|Default|GetTagRemovedAction))
          (bind _asynctaskproxy (Ability|Tasks|WaitGameplayTagAddToActor _playerstate (Variables|Default|GetActivationTag))
            (:then
              (Variables|Default|SetTagAddedAction _asynctaskproxy)
              (bind _asynctaskproxy_2 (Ability|Tasks|WaitGameplayTagRemoveFromActor _playerstate (Variables|Default|GetActivationTag))
                (:then
                  (Variables|Default|SetTagRemovedAction _asynctaskproxy_2))
                (:Removed
                  ;[1 0.5537 0.06398] BeastGirl Activation Tag Changed
                  (CallFunction|OnActvationChanged false))))
            (:Added
              (CallFunction|OnActvationChanged true))))))))

(event Custom|OnActvationChanged
  (params (bIsActive :type Boolean))
  (bind _beasthudscene (Variables|Scenes|GetBeastHUDScene))
  ;[1 0.5537 0.06398] Add / Remove Beast Girl HUD
  (bind _updated_variable (Utilities|FlowControl|DeltaGateWildcard (Variables|Default|GetIsTagActive) bIsActive)
    (:Success
      (if _updated_variable
        (DynamicUIDirector|AddScene _beasthudscene)
        (CallFunction|UpdateHUDElementsVisibility true)
        (else
          (DynamicUIDirector|RemoveScene _beasthudscene)
          (CallFunction|UpdateHUDElementsVisibility false))))))

(fn UpdateHUDElementsVisibility
  (params (bHideElements :type Boolean))
  (bind _returnvalue (DynamicUIDirector|GetOwningLocalPlayerController))
  (Utilities|IsValid _returnvalue
    (:Is%20Valid
      (bind _asfort_player_controller (Utilities|Casting|CastToFortPlayerController _returnvalue))
      (HUD|SetHUDElementVisibilitywithContext _asfort_player_controller (Variables|Default|GetHUDELements%20To%20Hide%20While%20In%20Form) bHideElements (GameplayTags|GetTagName (Variables|Default|GetActivationTag)) ""))))

(fn ConstructionScript)
