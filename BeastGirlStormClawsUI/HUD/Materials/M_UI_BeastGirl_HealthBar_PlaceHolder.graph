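# Translucent UI material that draws a placeholder health bar for the BeastGirl HUD.
# A Fill parameter added to inverted horizontal UV is floored into a hard mask, which lerps between background and fill for color and opacity.
# The fill color blends two color parameters along normalized UI UV, with a slow time-driven sine pulsing the gradient.
# Named reroutes carry the bar color, background color and background opacity into the blends.
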Material /BeastGirlStormClawsUI/HUD/Materials/M_UI_BeastGirl_HealthBar_PlaceHolder
graph {
 properties {bAutomaticallySetUsageInEditor=false, blendMode=BLEND_Translucent, materialDomain=MD_UI, usedShadingModels=Default Lit}
 nodes [
  n1 = TextureCoordinate()
  n2 = Add(A=OneMinus(None=ComponentMask(None=@n1) {r=true}), B=ScalarParameter() {defaultValue=1, parameterName=Fill})
  n3 = Floor(None=@n2)
  LinearInterpolate(A=NamedRerouteUsage(), B=Multiply(A=@n3, B=NamedRerouteUsage()), Alpha=@n3)
  LinearInterpolate(A=NamedRerouteUsage(), B=@n3, Alpha=@n3)
  n4 = VectorParameter() {parameterName=ProgressColorStart, defaultValue=<rgba 0.2705 0.0467 0.9911 1>}
  n5 = VectorParameter() {parameterName=ProgressColorStop, defaultValue=<rgba 0.7605 0.0762 1 1>}
  n6 = MaterialFunctionCall() {materialFunction=&GetUserInterfaceUV}
  n7 = Sine(None=Multiply(A=Time()) {constB=0.4})
  n8 = Saturate(None=Multiply(A=ComponentMask(None=@n6.Normalized UV) {r=true}, B=ConstantBiasScale(None=@n7) {bias=3, scale=0.25}))
  NamedRerouteDeclaration(None=Saturate(None=LinearInterpolate(A=@n4.RGB, B=@n5.RGB, Alpha=@n8))) {name=ProgressBarColor, nodeColor=<rgba 0 0 0 1>}
  NamedRerouteDeclaration(None=@n2) {name=ProgressBarBackground, nodeColor=<rgba 0 0 0 1>}
  n9 = VectorParameter() {parameterName=BackgroundColor, defaultValue=<rgba 0 0 0 0.5>}
  NamedRerouteDeclaration(None=Multiply(A=NamedRerouteUsage(), B=@n9.RGB)) {name=ProgressBarBackgroundColor, nodeColor=<rgba 0 0 0 1>}
  NamedRerouteDeclaration(None=@n9.A) {name=BackgroundOpacity, nodeColor=<rgba 0 0 0 1>}
 ]
}
