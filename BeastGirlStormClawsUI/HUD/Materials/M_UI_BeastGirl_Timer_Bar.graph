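# Translucent UI material for a horizontal timer bar in the BeastGirl Storm Claws HUD.
# A Step on the U texture coordinate against the Fill parameter masks a two-color horizontal gradient, blended over a background color.
# A second Step forms the backing mask, with named reroutes sharing both masks.
# Opacity math takes the maximum of a mask and a BackingTransparency-scaled mask, then applies MasterOpacity.
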Material /BeastGirlStormClawsUI/HUD/Materials/M_UI_BeastGirl_Timer_Bar
graph {
 properties {bAutomaticallySetUsageInEditor=false, blendMode=BLEND_Translucent, materialDomain=MD_UI, usedShadingModels=Default Lit}
 nodes [
  n1 = VectorParameter() {parameterName=FillColorStart, defaultValue=<rgba 0.7605 0.0762 1 1>}
  n2 = VectorParameter() {parameterName=FillColorEnd, defaultValue=<rgba 0.2705 0.0467 0.9911 1>}
  n3 = ComponentMask(None=TextureCoordinate()) {r=true}
  n4 = Step(Y=@n3, X=ScalarParameter() {defaultValue=1, parameterName=Fill}) {constX=0.5}
  LinearInterpolate(
   A=Multiply(A=VectorParameter() {parameterName=BackgroundColor, defaultValue=<rgba 0 0 0 1>}.RGB, B=NamedRerouteUsage())
   B=Multiply(A=LinearInterpolate(A=@n1.RGB, B=@n2.RGB, Alpha=@n3), B=NamedRerouteDeclaration(None=@n4) {name=Fill, nodeColor=<rgba 0.2187 1 0 1>})
   Alpha=NamedRerouteUsage()
  )
  n5 = NamedRerouteUsage()
  n6 = ScalarParameter() {defaultValue=0.5, parameterName=BackingTransparency}
  Multiply(A=Max(A=NamedRerouteUsage(), B=Multiply(A=@n5, B=@n6)), B=ScalarParameter() {defaultValue=1, parameterName=MasterOpacity})
  NamedRerouteDeclaration(None=Step(Y=@n3)) {name=Backing, nodeColor=<rgba 0.1016 0 1 1>}
 ]
}
