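# Translucent UI material for a placeholder BeastGirl HUD timer, drawn as a radial countdown ring over a circular background.
# Anti-aliased SDF circles shape the background disc, and subtracting an inner circle forms the ring, both scaled by TimerSize.
# Rotated UVs become a radial value that is offset by RadialFill and floored for a hard sweep, while RadialFill also blends ending-to-start color.
# Named reroutes carry each layer's color and opacity into chained lerps.
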Material /BeastGirlStormClawsUI/HUD/Materials/M_UI_BeastGirl_Timer_PlaceHolder
graph {
 properties {bAutomaticallySetUsageInEditor=false, blendMode=BLEND_Translucent, materialDomain=MD_UI, usedShadingModels=Default Lit}
 nodes [
  n1 = NamedRerouteUsage()
  n2 = LinearInterpolate(A=NamedRerouteUsage(), B=@n1, Alpha=@n1)
  LinearInterpolate(A=NamedRerouteUsage(), B=LinearInterpolate(A=NamedRerouteUsage(), B=NamedRerouteUsage(), Alpha=NamedRerouteUsage()), Alpha=@n2)
  Multiply(A=LinearInterpolate(A=NamedRerouteUsage(), B=@n2, Alpha=@n2), B=ScalarParameter() {defaultValue=1, parameterName=TimerOpacity})
  n3 = ScalarParameter() {defaultValue=1, parameterName=TimerSize}
  n4 = Subtract(A=@n3, B=ScalarParameter() {defaultValue=0.45, parameterName=RingThickness})
  n5 = MaterialFunctionCall(DistanceField=MaterialFunctionCall!(
    DistanceFieldA=MaterialFunctionCall!(UVs, Radius=@n3) {materialFunction=/FortUILibrary/MaterialFunctions/Shapes/SDFs/MF_SDF_Circle}.SDF
    DistanceFieldB=MaterialFunctionCall(Radius=@n4) {materialFunction=&MF_SDF_Circle}.SDF
    BlendAmount
   ) {materialFunction=/FortUILibrary/MaterialFunctions/Shapes/Modifiers/MF_UI_SDF_Subtraction}.Exact) {materialFunction=&MF_UI_AntiAliasing}
  NamedRerouteDeclaration(None=Multiply(A=VectorParameter() {parameterName=RingColor, defaultValue=<rgba 0 0 0 0.5>}.RGB, B=@n5.Inner)) {name=RingFill, nodeColor=<rgba 0 0 0 1>}
  n6 = MaterialFunctionCall(DistanceField=MaterialFunctionCall(Radius=ScalarParameter() {defaultValue=1, parameterName=TimerSize}) {materialFunction=&MF_SDF_Circle}.SDF) {materialFunction=&MF_UI_AntiAliasing}
  NamedRerouteDeclaration(None=Multiply(A=VectorParameter() {parameterName=BackgroundColor, defaultValue=<rgba 0.0156 0.0131 0.0122 1>}.RGB, B=@n6.Inner)) {name=Background, nodeColor=<rgba 0 0 0 1>}
  NamedRerouteDeclaration(None=Multiply(A=@n6.Inner, B=ScalarParameter() {parameterName=BackgroundOpacity})) {name=BackgroundOpacity, nodeColor=<rgba 0 0 0 1>}
  NamedRerouteDeclaration(None=Multiply(A=@n5.Inner, B=ScalarParameter() {defaultValue=0.8, parameterName=RingOpacity})) {name=RingOpacity, nodeColor=<rgba 0 0 0 1>}
  n7 = LinearInterpolate(
   A=VectorParameter() {parameterName=TimeEndingColor, defaultValue=<rgba 1 0 0.0204 1>}.RGB
   B=VectorParameter() {parameterName=TimeStartColor, defaultValue=<rgba 0.0879 0 0.151 1>}.RGB
   Alpha=ScalarParameter() {defaultValue=0.5, parameterName=RadialFill, sliderMax=1}
  )
  n8 = ScalarParameter() {defaultValue=0.5, parameterName=RadialFill, sliderMax=1}
  n9 = MaterialFunctionCall(UVs=TextureCoordinate(), "Rotation Angle (0-1)"=Constant() {r=0.25}) {materialFunction=&CustomRotator}
  n10 = Constant2Vector() {g=2, r=2}
  n11 = Constant2Vector() {g=1, r=1}
  n12 = MaterialFunctionCall(Vector or UVs=Multiply(A=Subtract(A=@n9.Rotated Values) {constB=0.5}, B=Multiply(A=@n10, B=@n11)) {constB=2}) {materialFunction=&VectorToRadialValue}
  n13 = ScalarParameter() {defaultValue=1, parameterName=TimerSize}
  n14 = Subtract(A=@n13, B=ScalarParameter() {defaultValue=0.45, parameterName=RingThickness})
  n15 = MaterialFunctionCall(DistanceField=MaterialFunctionCall(
    DistanceFieldA=MaterialFunctionCall(Radius=@n13) {materialFunction=&MF_SDF_Circle}.SDF
    DistanceFieldB=MaterialFunctionCall(Radius=@n14) {materialFunction=&MF_SDF_Circle}.SDF
   ) {materialFunction=&MF_UI_SDF_Subtraction}.Exact) {materialFunction=&MF_UI_AntiAliasing}
  n16 = Multiply(A=Floor(None=Add(A=@n8, B=@n12.Radial Coordinates)), B=@n15.Inner)
  NamedRerouteDeclaration(None=LinearInterpolate(A=@n7, B=@n7, Alpha=@n16)) {name=RadialTimerFill, nodeColor=<rgba 0 0 0 1>}
  NamedRerouteDeclaration(None=Max(A=Multiply(A=@n15.Inner) {constB=0}, B=@n16)) {name=RadialTimerOpacity, nodeColor=<rgba 0 0 0 1>}
 ]
}
